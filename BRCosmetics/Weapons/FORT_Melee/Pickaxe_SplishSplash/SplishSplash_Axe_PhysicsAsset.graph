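# Physics asset for the SplishSplash pickaxe cosmetic, defining a dangling duck on a string.
# Small capsules on four string bones and the duck form a constraint chain from a base string body, with limited swing and twist limits that widen toward the duck.
# Larger capsules on both thighs and calves, joined by free-rotation constraints, add character leg bodies, likely as collision for the dangling chain.
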PhysicsAsset /BRCosmetics/Weapons/FORT_Melee/Pickaxe_SplishSplash/SplishSplash_Axe_PhysicsAsset
bodies {
 base_String_01 [{shapeType=Capsule, length=0.505, radius=0.505, shapeName=base_String_01_capsule, rotation=<rot 0 -90 0>}]
 calf_l [{shapeType=Capsule, length=30, radius=9.5, shapeName=thigh_l_capsule, center=<-28.7961 0 -0.0225>, rotation=<rot 90 180 0>}]
 calf_r [{shapeType=Capsule, length=30, radius=9.5, shapeName=thigh_r_capsule, center=<28.7961 0 0.0225>, rotation=<rot 90 0 0>}]
 dyn_Duck [{shapeType=Capsule, length=2, radius=3.5, shapeName=dyn_Duck_capsule, center=<2.5 0 0>, rotation=<rot 90 0 0>}]
 dyn_String_01 [{shapeType=Capsule, length=2, radius=2.5, shapeName=dyn_String_01_capsule, center=<2 0 0>, rotation=<rot 90 0 0>}]
 dyn_String_02 [{shapeType=Capsule, length=2, radius=2.5, shapeName=dyn_String_02_capsule, center=<2 0 0>, rotation=<rot 90 0 0>}]
 dyn_String_03 [{shapeType=Capsule, length=2, radius=2.5, shapeName=dyn_String_03_capsule, center=<2 0 0>, rotation=<rot 90 0 0>}]
 dyn_String_04 [{shapeType=Capsule, length=2, radius=2.5, shapeName=dyn_String_04_capsule, center=<2 0 0>, rotation=<rot 90 0 0>}]
 thigh_l [{shapeType=Capsule, length=34.6616, radius=9.5479, shapeName=thigh_l_capsule, center=<-22.1907 -0.6066 0.3009>, rotation=<rot 90 180 0>}]
 thigh_r [{shapeType=Capsule, length=34.6616, radius=9.5479, shapeName=thigh_r_capsule, center=<22.1907 0.6066 -0.3009>, rotation=<rot 90 0 0>}]
}
constraints [
 {bone1Name=dyn_String_01, bone2Name=base_String_01, swing1LimitDegrees=25, swing2LimitDegrees=25, twistLimitDegrees=0}
 {bone1Name=dyn_String_02, bone2Name=dyn_String_01, swing1LimitDegrees=30, swing2LimitDegrees=30, twistLimitDegrees=3}
 {bone1Name=dyn_String_03, bone2Name=dyn_String_02, swing1LimitDegrees=25, swing2LimitDegrees=25, twistLimitDegrees=6}
 {bone1Name=dyn_String_04, bone2Name=dyn_String_03, swing1LimitDegrees=25, swing2LimitDegrees=25, twistLimitDegrees=8}
 {bone1Name=dyn_Duck, bone2Name=dyn_String_04, swing1LimitDegrees=25, swing2LimitDegrees=25, twistLimitDegrees=10}
 {
  bone1Name=calf_r
  bone2Name=thigh_r
  swing1LimitDegrees=45
  swing1Motion=Free
  swing2LimitDegrees=45
  swing2Motion=Free
  twistLimitDegrees=45
  twistMotion=Free
 }
 {
  bone1Name=calf_l
  bone2Name=thigh_l
  swing1LimitDegrees=45
  swing1Motion=Free
  swing2LimitDegrees=45
  swing2Motion=Free
  twistLimitDegrees=45
  twistMotion=Free
 }
]
